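# Unlit masked Niagara mesh-particle material that animates a Pivot Painter mesh breaking apart, with intro, outro and shockwave phases.
# UV channels supply per-element pivots and random values, and vertex color supplies rotation axes, driving gradient-based offsets, pushes and spins around an instance-space ball position.
# A dynamic parameter drives intro, suck-in, shockwave and reset.
# Particle color, scaled per phase, feeds emissive through MF_FX_EmissiveAdjustments.
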Material /DealershipSharedLibrary/Assets/Sets/Break/M_DLR_PivotPainter_Break
graph {
 properties {
  bAutomaticallySetUsageInEditor=false
  bHasPixelAnimation=true
  bUsedWithNiagaraMeshParticles=true
  blendMode=BLEND_Masked
  shadingModel=MSM_Unlit
  twoSided=true
  usedShadingModels=Unlit
 }
 nodes [
  n1 = Saturate(None=NamedRerouteUsage())
  n2 = OneMinus(None=@n1)
  Add(
   A=Multiply(A=NamedRerouteUsage(), B=NamedRerouteDeclaration(None=@n2) {name=Reset Invert, nodeColor=<rgba 1 0.5625 0 1>})
   B=NamedRerouteUsage()
  )
  n3 = NamedRerouteUsage()
  n4 = NamedRerouteUsage()
  n5 = Multiply(
   A=Subtract(A=Add(A=@n3, B=@n4), B=ScalarParameter() {defaultValue=-512, group=Intro, parameterName=Vertex Bright Amount})
   B=ScalarParameter() {defaultValue=0.002, group=Intro, parameterName=Vertex Falloff}
  )
  n6 = Saturate(None=Sine(None=Saturate(None=@n5)))
  n7 = NamedRerouteUsage()
  n8 = ScalarParameter() {defaultValue=-0.002, group=Intro, parameterName=IntroElementOffset}
  n9 = NamedRerouteUsage()
  n10 = NamedRerouteDeclaration(None=Saturate(None=OneMinus(None=@n9))) {name=Intro Opacity, nodeColor=<rgba 0.1953 0 1 1>}
  n11 = Add(A=@n9, B=NamedRerouteUsage())
  n12 = ScalarParameter() {defaultValue=-128, group=Intro, parameterName=Bright Amount}
  n13 = Saturate(None=Multiply(A=Subtract(A=@n11, B=@n12), B=ScalarParameter() {defaultValue=0.01, group=Intro, parameterName=BrightFalloff}))
  n14 = Add(
   A=Multiply(A=NamedRerouteDeclaration(None=@n6) {name=Intro Vertex Offset, nodeColor=<rgba 1 0.0938 0 1>}, B=Multiply(A=@n7, B=@n8))
   B=Multiply(A=Multiply(A=@n10, B=@n13), B=NamedRerouteUsage())
  )
  n15 = Constant()
  n16 = VertexColor()
  n17 = Add(
   A=VectorParameter() {group=Outro, parameterName=PushAmount, defaultValue=<rgba -1000 0 -1000 1>}.RGBA
   B=Multiply(
    A=MaterialFunctionCall(R=@n16.R, G=@n16.R, B=@n16.R, A=@n16.R) {materialFunction=&AppendMany}.RGBA
    B=ScalarParameter() {defaultValue=2000, group=Outro, parameterName=Pushout}
   )
  )
  n18 = Add(A=Multiply(A=NamedRerouteUsage(), B=Constant() {r=64}), B=NamedRerouteUsage())
  n19 = ScalarParameter() {group=Outro, parameterName=OutroGraidentOffset}
  n20 = OneMinus(None=Subtract(A=@n18, B=@n19))
  n21 = ScalarParameter() {defaultValue=0.0005, group=Outro, parameterName=OutroGradientFalloff}
  n22 = Saturate(None=Multiply(A=@n20, B=@n21))
  n23 = Subtract(A=@n18, B=Add(A=@n19, B=ScalarParameter() {group=Outro, parameterName=OutroRotationGradientOffset}))
  n24 = Saturate(None=Multiply(A=OneMinus(None=@n23), B=@n21))
  n25 = ScalarParameter() {defaultValue=3, group=Rotate, parameterName=Rotate Amount}
  n26 = Subtract(A=NamedRerouteUsage(), B=Constant() {r=0.6})
  n27 = LinearInterpolate(A=Abs(None=@n25), B=@n25, Alpha=Ceil(None=Clamp(None=@n26) {maxDefault=200}))
  n28 = Multiply(A=Frac(None=Multiply(A=@n24, B=@n27)), B=Constant() {r=6.28})
  n29 = Add(
   A=Transform(None=LinearInterpolate(A=@n15, B=@n17, Alpha=@n22)) {transformSourceType=TRANSFORMSOURCE_Instance}
   B=RotateAboutAxis(
    NormalizedRotationAxis=NamedRerouteUsage()
    RotationAngle=Clamp(None=@n28) {maxDefault=200}
    PivotPoint=NamedRerouteUsage()
    Position=WorldPosition().XYZ
   )
  )
  n30 = Multiply(
   A=Subtract(A=NamedRerouteUsage(), B=NamedRerouteUsage())
   B=ScalarParameter() {defaultValue=0.005, group=Shockwave, parameterName=Shockwave Amount}
  )
  n31 = Add(A=ScalarParameter() {defaultValue=-0.5, group=Shockwave, parameterName=Shockwave Length}, B=@n30)
  n32 = Saturate(None=OneMinus(None=Ceil(None=@n31)))
  n33 = Multiply(A=@n32, B=Saturate(None=@n30))
  LinearInterpolate(
   A=LinearInterpolate(A=NamedRerouteUsage(), B=Add(A=@n14, B=@n29), Alpha=NamedRerouteUsage())
   B=Multiply(A=NamedRerouteUsage(), B=OneMinus(None=@n33))
   Alpha=@n1
  )
  n34 = MaterialFunctionCall(Color=ReflectionVectorWS()) {materialFunction=&SplitComponents}
  n35 = MaterialFunctionCall(R=@n34.R, G=@n34.R, B=@n34.R, A=Constant()) {materialFunction=&AppendMany}
  n36 = Constant()
  n37 = ParticleColor()
  n38 = Multiply(A=@n37.RGBA, B=ScalarParameter() {defaultValue=8, group=Intro, parameterName=Intro Brightness})
  n39 = NamedRerouteUsage()
  n40 = Multiply(A=@n37.RGBA, B=ScalarParameter() {defaultValue=1, group=Outro, parameterName=End Brightness})
  n41 = Saturate(None=NamedRerouteUsage())
  n42 = Multiply(A=Add(A=Multiply(A=@n38, B=@n39), B=Multiply(A=@n40, B=@n41)), B=NamedRerouteUsage())
  n43 = Multiply(A=Multiply(A=@n37.RGBA, B=ScalarParameter() {defaultValue=8, group=Shockwave, parameterName=Shockwave Brightness}), B=NamedRerouteUsage())
  MaterialFunctionCall(Emissive=Add(A=Multiply(A=@n35.RGBA, B=@n36), B=Add(A=@n42, B=@n43))) {materialFunction=&MF_FX_EmissiveAdjustments}
  n44 = Normalize(VectorInput=ConstantBiasScale(None=VertexColor()) {bias=-0.5, scale=2})
  NamedRerouteDeclaration(None=Normalize(VectorInput=Transform(None=@n44) {transformSourceType=TRANSFORMSOURCE_Instance})) {name=X-Axis Vector, nodeColor=<rgba 1 0.0156 0.6498 1>}
  NamedRerouteDeclaration(None=Subtract(A=NamedRerouteUsage(), B=WorldPosition().XYZ)) {name=Scale, nodeColor=<rgba 0 0.7906 1 1>}
  n45 = ComponentMask(None=TextureCoordinate() {coordinateIndex=2}) {g=true}
  NamedRerouteDeclaration(None=Multiply(A=@n45, B=ScalarParameter() {group=Intro, parameterName=Random Amount})) {name=Random Values Custom, nodeColor=<rgba 0 0.7656 1 1>}
  NamedRerouteDeclaration(None=@n45) {name=PP Random Values TC2 MG, nodeColor=<rgba 1 0.9141 0 1>}
  n46 = Constant3Vector() {constant=<rgba 0 0 320 0>}
  NamedRerouteDeclaration(None=TransformPosition(None=@n46) {transformSourceType=TRANSFORMPOSSOURCE_Instance, transformType=TRANSFORMPOSSOURCE_World}) {name=Ball Position, nodeColor=<rgba 0.5938 0 1 1>}
  n47 = NamedRerouteUsage()
  n48 = WorldPosition()
  NamedRerouteDeclaration(None=Subtract(A=@n47, B=@n48.XYZ)) {name=Ball Pos - World Pos, nodeColor=<rgba 0 1 0.2734 1>}
  n49 = NamedRerouteUsage()
  n50 = Constant() {r=2048}
  n51 = Divide(A=Add(A=Distance(A=@n48.XYZ, B=@n47), B=Multiply(A=@n49, B=@n50)), B=NamedRerouteUsage())
  n52 = ScalarParameter() {defaultValue=12, group=Sucky, parameterName=SuckyGradFalloff}
  NamedRerouteDeclaration(None=Saturate(None=Power(Base=@n51, Exp=@n52))) {name=End Suck In, nodeColor=<rgba 1 0 0.0469 1>}
  n53 = AppendVector(A=TextureCoordinate() {coordinateIndex=1}, B=ComponentMask(None=TextureCoordinate() {coordinateIndex=2}) {r=true})
  NamedRerouteDeclaration(None=Distance(A=@n53, B=@n46)) {name=Dist Pivot Pos and Ball Pos, nodeColor=<rgba 0 0.6953 1 1>}
  n54 = ScalarParameter() {group=Intro, parameterName=IntroMin}
  n55 = ScalarParameter() {defaultValue=1300, group=Intro, parameterName=IntroMax}
  n56 = NamedRerouteUsage()
  NamedRerouteDeclaration(None=Subtract(A=NamedRerouteUsage(), B=LinearInterpolate(A=@n54, B=@n55, Alpha=@n56))) {name=Intro Transition Control, nodeColor=<rgba 0.2873 0.0079 1 1>}
  NamedRerouteDeclaration(None=@n20) {name=End Transition Gradient Control, nodeColor=<rgba 0 1 0.3438 1>}
  NamedRerouteDeclaration(None=@n32) {name=Shockwave Add Length, nodeColor=<rgba 1 0.0234 0 1>}
  n57 = DynamicParameter() {defaultValue=<rgba 1 2048 256 0>, paramNames [Intro Transition, Sucky Grad Size, Shockwave Transition, Reset]}
  NamedRerouteDeclaration(None=@n57.Intro Transition) {name=DP Intro Transition, nodeColor=<rgba 1 1 1 1>}
  NamedRerouteDeclaration(None=@n57.Sucky Grad Size) {name=DP Sucky Grad Size, nodeColor=<rgba 1 1 1 1>}
  NamedRerouteDeclaration(None=@n57.Shockwave Transition) {name=DP Shockwave Transition, nodeColor=<rgba 1 1 1 1>}
  NamedRerouteDeclaration(None=@n57.Reset) {name=DP Reset, nodeColor=<rgba 1 1 1 1>}
  n58 = Constant3Vector() {constant=<rgba 0 -1 0 0>}
  NamedRerouteDeclaration(None=TransformPosition(None=Add(A=@n53, B=@n58)) {transformSourceType=TRANSFORMPOSSOURCE_Instance, transformType=TRANSFORMPOSSOURCE_World}) {name=NewPivotPosition, nodeColor=<rgba 0 1 0.8594 1>}
 ]
}
